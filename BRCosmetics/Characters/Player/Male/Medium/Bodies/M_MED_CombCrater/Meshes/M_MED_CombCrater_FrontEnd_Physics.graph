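# Physics asset for the front-end mesh of the M_MED_CombCrater male player body, driving a seven-chain dynamic coat.
# Capsule bodies cover the coat bones, pelvis, spine, legs and feet, plus a huge root capsule below the origin that likely approximates a floor.
# Angular limits shape each chain, with chains 2–6 linked to the pelvis.
# Angularly free cross-constraints tie neighboring chains together so the coat moves as a connected sheet.
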PhysicsAsset /BRCosmetics/Characters/Player/Male/Medium/Bodies/M_MED_CombCrater/Meshes/M_MED_CombCrater_FrontEnd_Physics
bodies {
 calf_l [{
   shapeType=Capsule
   length=27.4065
   radius=7.6315
   shapeName=calf_l_capsule
   center=<-22.0486 -0.8801 -0.3558>
   rotation=<rot 0.6284 85.5816 -93.5668>
  }]
 calf_r [{
   shapeType=Capsule
   length=27.5556
   radius=7.5972
   shapeName=calf_r_capsule
   center=<21.9042 0.795 0.2667>
   rotation=<rot 0.6158 86.0787 -93.4951>
  }]
 dyn_coat_1_1 [{
   shapeType=Capsule
   length=11.8383
   radius=4.3585
   shapeName=dyn_coat_1_1_capsule
   center=<8.6738 -0.0903 0.3326>
   rotation=<rot 48.3758 89.0766 -91.1936>
  }]
 dyn_coat_1_2 [{
   shapeType=Capsule
   length=11.4681
   radius=4.9825
   shapeName=dyn_coat_1_2_capsule
   center=<6.8333 -0.6362 1.3352>
   rotation=<rot 65.2854 92.1464 85.5467>
  }]
 dyn_coat_1_3 [{
   shapeType=Capsule
   length=16.0366
   radius=5.3363
   shapeName=dyn_coat_1_3_capsule
   center=<12.3364 -0.9517 5.6719>
   rotation=<rot 23.9947 91.4757 -59.9078>
  }]
 dyn_coat_1_4 [{
   shapeType=Capsule
   length=10.7777
   radius=6.8191
   shapeName=dyn_coat_1_4_capsule
   center=<9.2911 1.3293 9.885>
   rotation=<rot 29.0831 90.5456 2.5183>
  }]
 dyn_coat_1_5 [{
   shapeType=Capsule
   length=16.6617
   radius=8.1433
   shapeName=dyn_coat_1_5_capsule
   center=<16.0767 -4.683 8.5987>
   rotation=<rot 56.0005 96.9745 3.7496>
  }]
 dyn_coat_1_base [{shapeType=Capsule, length=0, radius=2.75, shapeName=dyn_coat_1_base_capsule, rotation=<rot 0 -90 0>}]
 dyn_coat_2_1 [{
   shapeType=Capsule
   length=15.25
   radius=3.505
   shapeName=dyn_coat_2_1_capsule
   center=<7.2242 0.2289 1.153>
   rotation=<rot -73.5738 -86.1823 -11.8483>
  }]
 dyn_coat_2_2 [{
   shapeType=Capsule
   length=13.6992
   radius=6.2548
   shapeName=dyn_coat_2_2_capsule
   center=<10.4655 -0.78 1.5813>
   rotation=<rot 84.3893 -84.5663 -167.356>
  }]
 dyn_coat_2_3 [{
   shapeType=Capsule
   length=20.255
   radius=8.255
   shapeName=dyn_coat_2_3_capsule
   center=<8.6873 -0.249 1.5437>
   rotation=<rot -80.7326 98.6399 164.9257>
  }]
 dyn_coat_2_4 [{
   shapeType=Capsule
   length=19.939
   radius=8.1501
   shapeName=dyn_coat_2_4_capsule
   center=<7.7339 -2.3161 2.7815>
   rotation=<rot 54.6327 -85.8303 -171.6633>
  }]
 dyn_coat_2_base [{
   shapeType=Capsule
   length=0
   radius=4.6465
   shapeName=dyn_coat_2_base_capsule
   center=<5.3177 -1.9838 -0.2132>
   rotation=<rot 0 -76.9928 -88.3516>
  }]
 dyn_coat_3_1 [{
   shapeType=Capsule
   length=11.505
   radius=4.505
   shapeName=dyn_coat_3_1_capsule
   center=<8.3105 0.653 1.3163>
   rotation=<rot 1.5848 2.4359 123.2083>
  }]
 dyn_coat_3_2 [{
   shapeType=Capsule
   length=14.3258
   radius=6.6718
   shapeName=dyn_coat_3_2_capsule
   center=<9.7394 1.3313 2.2586>
   rotation=<rot 3.1992 0.7581 -60.2157>
  }]
 dyn_coat_3_3 [{
   shapeType=Capsule
   length=24.6741
   radius=7.3263
   shapeName=dyn_coat_3_3_capsule
   center=<9.9217 0.1372 3.8886>
   rotation=<rot -0.838 -1.6375 -81.9847>
  }]
 dyn_coat_3_4 [{
   shapeType=Capsule
   length=24.3763
   radius=10.5752
   shapeName=dyn_coat_3_4_capsule
   center=<13.2757 1.61 2.1859>
   rotation=<rot 2.1454 177.6823 -127.7227>
  }]
 dyn_coat_3_base [{
   shapeType=Capsule
   length=0
   radius=2.8864
   shapeName=dyn_coat_3_base_capsule
   center=<0.1443 0.0628 0.1985>
   rotation=<rot 0 -66.3174 -82.3122>
  }]
 dyn_coat_4_1 [{
   shapeType=Capsule
   length=18.9427
   radius=5.4492
   shapeName=dyn_coat_4_1_capsule
   center=<9.6993 0.0286 -0.3204>
   rotation=<rot 7.985 -0.402 -92.7957>
  }]
 dyn_coat_4_2 [{
   shapeType=Capsule
   length=18.63
   radius=7.2737
   shapeName=dyn_coat_4_2_capsule
   center=<11.5928 0.145 -3.4501>
   rotation=<rot 4.5135 -0.2617 -89.2583>
  }]
 dyn_coat_4_3 [{
   shapeType=Capsule
   length=31.6574
   radius=8.8995
   shapeName=dyn_coat_4_3_capsule
   center=<9.8941 -2.0522 -3.063>
   rotation=<rot 0.3298 0.3242 131.6572>
  }]
 dyn_coat_4_4 [{
   shapeType=Capsule
   length=39.0247
   radius=9.9722
   shapeName=dyn_coat_4_4_capsule
   center=<10.4974 -0.0853 -0.9058>
   rotation=<rot 1.1467 -1.43 -70.0748>
  }]
 dyn_coat_4_base [{shapeType=Capsule, length=0, radius=2.75, shapeName=dyn_coat_4_base_capsule, rotation=<rot 0 -90 0>}]
 dyn_coat_5_1 [{
   shapeType=Capsule
   length=12.2721
   radius=4.5003
   shapeName=dyn_coat_5_1_capsule
   center=<9.4339 -0.9315 0.7843>
   rotation=<rot 0.9218 -0.8419 -170.7222>
  }]
 dyn_coat_5_2 [{
   shapeType=Capsule
   length=12.4202
   radius=7.3132
   shapeName=dyn_coat_5_2_capsule
   center=<8.1112 -3.4449 1.4794>
   rotation=<rot -0.8663 -0.1117 -177.575>
  }]
 dyn_coat_5_3 [{
   shapeType=Capsule
   length=21.6771
   radius=9.3901
   shapeName=dyn_coat_5_3_capsule
   center=<14.1714 -3.8146 0.4379>
   rotation=<rot -0.7839 5.3207 -11.8954>
  }]
 dyn_coat_5_4 [{
   shapeType=Capsule
   length=26.8025
   radius=10.102
   shapeName=dyn_coat_5_4_capsule
   center=<11.953 -3.0393 -0.7375>
   rotation=<rot -1.714 -1.6004 141.512>
  }]
 dyn_coat_5_base [{shapeType=Capsule, length=0, radius=2.75, shapeName=dyn_coat_5_base_capsule, rotation=<rot 0 -90 0>}]
 dyn_coat_6_1 [{
   shapeType=Capsule
   length=13.647
   radius=4.6976
   shapeName=dyn_coat_6_1_capsule
   center=<8.3409 -0.146 -0.6544>
   rotation=<rot -61.1505 79.9039 -3.128>
  }]
 dyn_coat_6_2 [{
   shapeType=Capsule
   length=18.3535
   radius=6.4844
   shapeName=dyn_coat_6_2_capsule
   center=<10.4472 -0.3321 -2.0936>
   rotation=<rot 0.2871 3.3172 74.8929>
  }]
 dyn_coat_6_3 [{
   shapeType=Capsule
   length=26.25
   radius=7.75
   shapeName=dyn_coat_6_3_capsule
   center=<9.759 -0.0817 -3.4919>
   rotation=<rot -0.7215 4.0175 -109.1267>
  }]
 dyn_coat_6_4 [{
   shapeType=Capsule
   length=20.9387
   radius=7.5608
   shapeName=dyn_coat_6_4_capsule
   center=<9.9459 0.2326 -3.4008>
   rotation=<rot 1.4986 -0.3865 76.8211>
  }]
 dyn_coat_6_base [{shapeType=Capsule, length=0, radius=2.75, shapeName=dyn_coat_6_base_capsule, rotation=<rot 0 -90 0>}]
 dyn_coat_7_1 [{
   shapeType=Capsule
   length=6.5011
   radius=4.1534
   shapeName=dyn_coat_7_1_capsule
   center=<11.3242 0.9507 -2.315>
   rotation=<rot 27.7002 -80.3585 -173.576>
  }]
 dyn_coat_7_2 [{
   shapeType=Capsule
   length=5.8365
   radius=4.8762
   shapeName=dyn_coat_7_2_capsule
   center=<8.2309 -0.8881 -3.7451>
   rotation=<rot -3.5693 1.1264 -129.2055>
  }]
 dyn_coat_7_3 [{
   shapeType=Capsule
   length=8.9524
   radius=6.5644
   shapeName=dyn_coat_7_3_capsule
   center=<10.0661 -1.8474 -4.9242>
   rotation=<rot -1.9997 2.0906 -137.7883>
  }]
 dyn_coat_7_4 [{
   shapeType=Capsule
   length=11.5588
   radius=6.7771
   shapeName=dyn_coat_7_4_capsule
   center=<8.6178 -1.5624 -6.3574>
   rotation=<rot 2.5495 174.1548 -40.155>
  }]
 dyn_coat_7_5 [{
   shapeType=Capsule
   length=10.1339
   radius=8.2453
   shapeName=dyn_coat_7_5_capsule
   center=<8.7509 -2.7477 -7.1239>
   rotation=<rot 4.1128 175.9362 141.2108>
  }]
 dyn_coat_7_base [{shapeType=Capsule, length=0, radius=2.75, shapeName=dyn_coat_7_base_capsule, rotation=<rot 0 -90 0>}]
 foot_l [{
   shapeType=Capsule
   length=21.6706
   radius=11.074
   shapeName=foot_l_capsule
   center=<-3.78 6.7675 -0.0561>
   rotation=<rot -0.167 3.46 -89.3768>
  }]
 foot_r [{
   shapeType=Capsule
   length=21.6695
   radius=11.1373
   shapeName=foot_r_capsule
   center=<3.7075 -6.7572 0.0557>
   rotation=<rot -0.1677 3.5482 -89.3741>
  }]
 pelvis [{
   shapeType=Capsule
   length=10.2223
   radius=10.473
   shapeName=pelvis_capsule
   center=<-0.8136 -0.325 -0.3993>
   rotation=<rot 0.033 -90.2071 -0.4405>
  }]
 root [{shapeType=Capsule, length=0, radius=214.005, shapeName=root_capsule, center=<0 -1.5038 -205.0303>, rotation=<rot 0 -90 0>}]
 spine_01 [{
   shapeType=Capsule
   length=14.6869
   radius=9.8102
   shapeName=spine_01_capsule
   center=<5.0973 1.217 0.1467>
   rotation=<rot -3.6196 -90 -0.0922>
  }]
 spine_04 [{
   shapeType=Capsule
   length=11.4062
   radius=11.8361
   shapeName=spine_04_capsule
   center=<-4.4366 -0.6386 0.0676>
   rotation=<rot -1.0344 -90 -0.8334>
  }]
 thigh_l [{
   shapeType=Capsule
   length=31.099
   radius=8.8286
   shapeName=thigh_l_capsule
   center=<-19.6949 -0.9531 0.0155>
   rotation=<rot 0 94.3067 -89.452>
  }]
 thigh_r [{
   shapeType=Capsule
   length=32.2967
   radius=8.8033
   shapeName=thigh_r_capsule
   center=<19.6656 1.5792 0.4701>
   rotation=<rot -0.0001 91.0489 -90.7869>
  }]
}
constraints [
 {bone1Name=dyn_coat_2_1, bone2Name=dyn_coat_2_base, swing1LimitDegrees=30, swing2LimitDegrees=30, twistLimitDegrees=30}
 {bone1Name=dyn_coat_2_2, bone2Name=dyn_coat_2_1, swing1LimitDegrees=15, swing2LimitDegrees=30, twistLimitDegrees=30}
 {bone1Name=dyn_coat_2_3, bone2Name=dyn_coat_2_2, swing1LimitDegrees=30, swing2LimitDegrees=45, twistLimitDegrees=30}
 {bone1Name=dyn_coat_2_4, bone2Name=dyn_coat_2_3, swing1LimitDegrees=15, swing2LimitDegrees=45, twistLimitDegrees=30}
 {bone1Name=dyn_coat_3_1, bone2Name=dyn_coat_3_base, swing1LimitDegrees=25, swing2LimitDegrees=45, twistLimitDegrees=15}
 {bone1Name=dyn_coat_3_2, bone2Name=dyn_coat_3_1, swing1LimitDegrees=15, swing2LimitDegrees=45, twistLimitDegrees=5}
 {bone1Name=dyn_coat_3_3, bone2Name=dyn_coat_3_2, swing1LimitDegrees=25, swing2LimitDegrees=35, twistLimitDegrees=15}
 {bone1Name=dyn_coat_3_4, bone2Name=dyn_coat_3_3, swing1LimitDegrees=30, swing2LimitDegrees=45, twistLimitDegrees=30}
 {bone1Name=dyn_coat_4_1, bone2Name=dyn_coat_4_base, swing1LimitDegrees=30, swing2LimitDegrees=45, twistLimitDegrees=15}
 {bone1Name=dyn_coat_4_2, bone2Name=dyn_coat_4_1, swing1LimitDegrees=10, swing2LimitDegrees=45, twistLimitDegrees=5}
 {bone1Name=dyn_coat_4_3, bone2Name=dyn_coat_4_2, swing1LimitDegrees=15, swing2LimitDegrees=35, twistLimitDegrees=5}
 {bone1Name=dyn_coat_4_4, bone2Name=dyn_coat_4_3, swing1LimitDegrees=30, swing2LimitDegrees=45, twistLimitDegrees=30}
 {bone1Name=dyn_coat_5_2, bone2Name=dyn_coat_5_1, swing1LimitDegrees=25, swing2LimitDegrees=45, twistLimitDegrees=15}
 {bone1Name=dyn_coat_5_3, bone2Name=dyn_coat_5_2, swing1LimitDegrees=25, swing2LimitDegrees=45, twistLimitDegrees=15}
 {bone1Name=dyn_coat_5_4, bone2Name=dyn_coat_5_3, swing1LimitDegrees=30, swing2LimitDegrees=45, twistLimitDegrees=30}
 {bone1Name=dyn_coat_6_1, bone2Name=dyn_coat_6_base, swing1LimitDegrees=45, swing2LimitDegrees=30, twistLimitDegrees=30}
 {bone1Name=dyn_coat_6_2, bone2Name=dyn_coat_6_1, swing1LimitDegrees=30, swing2LimitDegrees=30, twistLimitDegrees=30}
 {bone1Name=dyn_coat_6_3, bone2Name=dyn_coat_6_2, swing1LimitDegrees=30, swing2LimitDegrees=45, twistLimitDegrees=30}
 {bone1Name=dyn_coat_6_4, bone2Name=dyn_coat_6_3, swing1LimitDegrees=30, swing2LimitDegrees=45, twistLimitDegrees=30}
 {bone1Name=dyn_coat_1_1, bone2Name=dyn_coat_1_base, swing1LimitDegrees=30, swing2LimitDegrees=30, twistLimitDegrees=45}
 {bone1Name=dyn_coat_1_2, bone2Name=dyn_coat_1_1, swing1LimitDegrees=30, swing2LimitDegrees=30, twistLimitDegrees=45}
 {bone1Name=dyn_coat_1_3, bone2Name=dyn_coat_1_2, swing1LimitDegrees=15, swing2LimitDegrees=45, twistLimitDegrees=30}
 {bone1Name=dyn_coat_1_4, bone2Name=dyn_coat_1_3, swing1LimitDegrees=15, swing2LimitDegrees=45, twistLimitDegrees=30}
 {bone1Name=dyn_coat_1_5, bone2Name=dyn_coat_1_4, swing1LimitDegrees=15, swing2LimitDegrees=45, twistLimitDegrees=45}
 {bone1Name=dyn_coat_7_1, bone2Name=dyn_coat_7_base, swing1LimitDegrees=30, swing2LimitDegrees=30, twistLimitDegrees=30}
 {bone1Name=dyn_coat_7_2, bone2Name=dyn_coat_7_1, swing1LimitDegrees=30, swing2LimitDegrees=30, twistLimitDegrees=45}
 {bone1Name=dyn_coat_7_3, bone2Name=dyn_coat_7_2, swing1LimitDegrees=20, swing2LimitDegrees=60, twistLimitDegrees=30}
 {bone1Name=dyn_coat_7_4, bone2Name=dyn_coat_7_3, swing1LimitDegrees=30, swing2LimitDegrees=45, twistLimitDegrees=30}
 {bone1Name=dyn_coat_7_5, bone2Name=dyn_coat_7_4, swing1LimitDegrees=30, swing2LimitDegrees=45, twistLimitDegrees=30}
 {
  bone1Name=dyn_coat_2_4
  bone2Name=dyn_coat_1_5
  swing1LimitDegrees=45
  swing1Motion=Free
  swing2LimitDegrees=45
  swing2Motion=Free
  twistLimitDegrees=45
  twistMotion=Free
 }
 {
  bone1Name=dyn_coat_2_4
  bone2Name=dyn_coat_3_4
  swing1LimitDegrees=45
  swing1Motion=Free
  swing2LimitDegrees=45
  swing2Motion=Free
  twistLimitDegrees=45
  twistMotion=Free
 }
 {
  bone1Name=dyn_coat_3_4
  bone2Name=dyn_coat_4_4
  swing1LimitDegrees=45
  swing1Motion=Free
  swing2LimitDegrees=45
  swing2Motion=Free
  twistLimitDegrees=45
  twistMotion=Free
 }
 {
  bone1Name=dyn_coat_5_4
  bone2Name=dyn_coat_4_4
  swing1LimitDegrees=45
  swing1Motion=Free
  swing2LimitDegrees=45
  swing2Motion=Free
  twistLimitDegrees=45
  twistMotion=Free
 }
 {
  bone1Name=dyn_coat_6_4
  bone2Name=dyn_coat_5_4
  swing1LimitDegrees=45
  swing1Motion=Free
  swing2LimitDegrees=45
  swing2Motion=Free
  twistLimitDegrees=45
  twistMotion=Free
 }
 {
  bone1Name=dyn_coat_7_5
  bone2Name=dyn_coat_6_4
  swing1LimitDegrees=45
  swing1Motion=Free
  swing2LimitDegrees=45
  swing2Motion=Free
  twistLimitDegrees=45
  twistMotion=Free
 }
 {
  bone1Name=dyn_coat_2_2
  bone2Name=dyn_coat_1_3
  swing1LimitDegrees=45
  swing1Motion=Free
  swing2LimitDegrees=45
  swing2Motion=Free
  twistLimitDegrees=45
  twistMotion=Free
 }
 {
  bone1Name=dyn_coat_6_2
  bone2Name=dyn_coat_7_3
  swing1LimitDegrees=45
  swing1Motion=Free
  swing2LimitDegrees=45
  swing2Motion=Free
  twistLimitDegrees=45
  twistMotion=Free
 }
 {bone1Name=dyn_coat_5_1, bone2Name=dyn_coat_5_base, swing1LimitDegrees=30, swing2LimitDegrees=30, twistLimitDegrees=5}
 {
  bone1Name=dyn_coat_5_2
  bone2Name=dyn_coat_6_2
  swing1LimitDegrees=45
  swing1Motion=Free
  swing2LimitDegrees=45
  swing2Motion=Free
  twistLimitDegrees=45
  twistMotion=Free
 }
 {
  bone1Name=dyn_coat_6_base
  bone2Name=pelvis
  swing1LimitDegrees=45
  swing1Motion=Free
  swing2LimitDegrees=45
  swing2Motion=Free
  twistLimitDegrees=45
  twistMotion=Free
 }
 {
  bone1Name=dyn_coat_5_base
  bone2Name=pelvis
  swing1LimitDegrees=45
  swing1Motion=Free
  swing2LimitDegrees=45
  swing2Motion=Free
  twistLimitDegrees=45
  twistMotion=Free
 }
 {
  bone1Name=dyn_coat_4_base
  bone2Name=pelvis
  swing1LimitDegrees=45
  swing1Motion=Free
  swing2LimitDegrees=45
  swing2Motion=Free
  twistLimitDegrees=45
  twistMotion=Free
 }
 {
  bone1Name=dyn_coat_3_base
  bone2Name=pelvis
  swing1LimitDegrees=45
  swing1Motion=Free
  swing2LimitDegrees=45
  swing2Motion=Free
  twistLimitDegrees=45
  twistMotion=Free
 }
 {
  bone1Name=dyn_coat_2_base
  bone2Name=pelvis
  swing1LimitDegrees=45
  swing1Motion=Free
  swing2LimitDegrees=45
  swing2Motion=Free
  twistLimitDegrees=45
  twistMotion=Free
 }
 {
  bone1Name=dyn_coat_1_4
  bone2Name=dyn_coat_2_3
  swing1LimitDegrees=45
  swing1Motion=Free
  swing2LimitDegrees=45
  swing2Motion=Free
  twistLimitDegrees=45
  twistMotion=Free
 }
 {
  bone1Name=dyn_coat_6_3
  bone2Name=dyn_coat_7_4
  swing1LimitDegrees=45
  swing1Motion=Free
  swing2LimitDegrees=45
  swing2Motion=Free
  twistLimitDegrees=45
  twistMotion=Free
 }
 {
  bone1Name=dyn_coat_6_3
  bone2Name=dyn_coat_5_3
  swing1LimitDegrees=45
  swing1Motion=Free
  swing2LimitDegrees=45
  swing2Motion=Free
  twistLimitDegrees=45
  twistMotion=Free
 }
 {
  bone1Name=dyn_coat_3_3
  bone2Name=dyn_coat_2_3
  swing1LimitDegrees=45
  swing1Motion=Free
  swing2LimitDegrees=45
  swing2Motion=Free
  twistLimitDegrees=45
  twistMotion=Free
 }
 {
  bone1Name=dyn_coat_2_2
  bone2Name=dyn_coat_3_2
  swing1LimitDegrees=45
  swing1Motion=Free
  swing2LimitDegrees=45
  swing2Motion=Free
  twistLimitDegrees=45
  twistMotion=Free
 }
 {
  bone1Name=dyn_coat_5_3
  bone2Name=dyn_coat_4_3
  swing1LimitDegrees=45
  swing1Motion=Free
  swing2LimitDegrees=45
  swing2Motion=Free
  twistLimitDegrees=45
  twistMotion=Free
 }
 {
  bone1Name=dyn_coat_3_3
  bone2Name=dyn_coat_4_3
  swing1LimitDegrees=45
  swing1Motion=Free
  swing2LimitDegrees=45
  swing2Motion=Free
  twistLimitDegrees=45
  twistMotion=Free
 }
 {
  bone1Name=dyn_coat_6_1
  bone2Name=dyn_coat_7_2
  swing1LimitDegrees=45
  swing1Motion=Free
  swing2LimitDegrees=45
  swing2Motion=Free
  twistLimitDegrees=45
  twistMotion=Free
 }
 {
  bone1Name=dyn_coat_1_2
  bone2Name=dyn_coat_2_1
  swing1LimitDegrees=45
  swing1Motion=Free
  swing2LimitDegrees=45
  swing2Motion=Free
  twistLimitDegrees=45
  twistMotion=Free
 }
]
