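# Physics asset for the male medium CombCrater player body, set up for dynamic motion of its coat.
# Seven chains of coat capsules use limited swing and very tight twist, with the bases of the middle five constrained to the pelvis.
# Angularly free cross-links join neighboring chains into an open, cloth-like lattice.
# Capsules on legs, feet, pelvis and spine provide body shapes, plus a huge root sphere below, likely a floor stand-in.
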PhysicsAsset /BRCosmetics/Characters/Player/Male/Medium/Bodies/M_MED_CombCrater/Meshes/M_MED_CombCrater_Physics
bodies {
 calf_l [{
   shapeType=Capsule
   length=26.4065
   radius=8.6315
   shapeName=calf_l_capsule
   center=<-22.0486 -0.8801 -0.3558>
   rotation=<rot 0.6284 85.5816 -93.5668>
  }]
 calf_r [{
   shapeType=Capsule
   length=27.5556
   radius=7.5972
   shapeName=calf_r_capsule
   center=<21.9042 0.795 0.2667>
   rotation=<rot 0.6158 86.0787 -93.4951>
  }]
 dyn_coat_1_1 [{shapeType=Capsule, length=5.5011, radius=5.1534, shapeName=dyn_coat_1_1_capsule, center=<11.3242 0.9507 2.315>, rotation=<rot 0 0.0002 -40>}]
 dyn_coat_1_2 [{shapeType=Capsule, length=4.8365, radius=5.8762, shapeName=dyn_coat_1_2_capsule, center=<8.2309 -0.8881 3.7451>, rotation=<rot 0 0.0002 -50>}]
 dyn_coat_1_3 [{
   shapeType=Capsule
   length=8.9524
   radius=6.5644
   shapeName=dyn_coat_1_3_capsule
   center=<10.0661 1.8474 4.9242>
   rotation=<rot -0.0002 0.0003 -20>
  }]
 dyn_coat_1_4 [{
   shapeType=Capsule
   length=11.0588
   radius=7.2771
   shapeName=dyn_coat_1_4_capsule
   center=<8.6178 1.5624 6.3574>
   rotation=<rot -0.0001 0.0002 -20>
  }]
 dyn_coat_1_5 [{
   shapeType=Capsule
   length=11.1339
   radius=7.2453
   shapeName=dyn_coat_1_5_capsule
   center=<8.7509 -2.7477 7.1239>
   rotation=<rot 0 0.0002 -40>
  }]
 dyn_coat_1_base [{shapeType=Capsule, length=0, radius=2, shapeName=dyn_coat_1_base_capsule}]
 dyn_coat_2_1 [{
   shapeType=Capsule
   length=13.647
   radius=4.6976
   shapeName=dyn_coat_2_1_capsule
   center=<8.3409 -0.146 0.6544>
   rotation=<rot -0.0003 0.0016 -70.0002>
  }]
 dyn_coat_2_2 [{
   shapeType=Capsule
   length=17.3535
   radius=7.4844
   shapeName=dyn_coat_2_2_capsule
   center=<10.4472 -0.3321 2.0936>
   rotation=<rot 0.0001 0.0003 -90>
  }]
 dyn_coat_2_3 [{shapeType=Capsule, length=26.25, radius=7.75, shapeName=dyn_coat_2_3_capsule, center=<9.759 -0.0817 3.4919>, rotation=<rot 0 0.0003 -80>}]
 dyn_coat_2_4 [{
   shapeType=Capsule
   length=19.9387
   radius=8.5608
   shapeName=dyn_coat_2_4_capsule
   center=<9.9459 0.2326 3.4008>
   rotation=<rot 0.0002 0.0002 69.9999>
  }]
 dyn_coat_2_base [{shapeType=Capsule, length=0, radius=2, shapeName=dyn_coat_2_base_capsule}]
 dyn_coat_3_1 [{
   shapeType=Capsule
   length=10.505
   radius=5.505
   shapeName=dyn_coat_3_1_capsule
   center=<8.3105 0.653 1.3163>
   rotation=<rot 1.5848 2.4359 123.2083>
  }]
 dyn_coat_3_2 [{
   shapeType=Capsule
   length=12.8258
   radius=8.1718
   shapeName=dyn_coat_3_2_capsule
   center=<9.7394 1.3313 2.2586>
   rotation=<rot 3.1992 0.7581 -60.2157>
  }]
 dyn_coat_3_3 [{
   shapeType=Capsule
   length=21.6741
   radius=10.3263
   shapeName=dyn_coat_3_3_capsule
   center=<9.9217 0.1371 3.8886>
   rotation=<rot -0.838 -1.6375 -81.9847>
  }]
 dyn_coat_3_4 [{
   shapeType=Capsule
   length=21.3763
   radius=13.5752
   shapeName=dyn_coat_3_4_capsule
   center=<13.1925 0.6757 0.9755>
   rotation=<rot 2.1454 177.6823 -127.7227>
  }]
 dyn_coat_3_base [{shapeType=Capsule, length=0, radius=2, shapeName=dyn_coat_3_base_capsule}]
 dyn_coat_4_1 [{
   shapeType=Capsule
   length=17.9427
   radius=6.4492
   shapeName=dyn_coat_4_1_capsule
   center=<9.6993 0.0286 -0.3204>
   rotation=<rot 7.985 -0.402 -92.7957>
  }]
 dyn_coat_4_2 [{
   shapeType=Capsule
   length=15.13
   radius=10.7737
   shapeName=dyn_coat_4_2_capsule
   center=<11.5928 0.145 -3.4501>
   rotation=<rot 4.5135 -0.2617 -89.2583>
  }]
 dyn_coat_4_3 [{
   shapeType=Capsule
   length=26.6574
   radius=13.8995
   shapeName=dyn_coat_4_3_capsule
   center=<9.8732 -0.3283 -1.1253>
   rotation=<rot 0.3298 0.3242 131.6572>
  }]
 dyn_coat_4_4 [{
   shapeType=Capsule
   length=31.5247
   radius=14.4722
   shapeName=dyn_coat_4_4_capsule
   center=<10.4834 0.3787 0.3725>
   rotation=<rot 1.1467 -1.43 -70.0748>
  }]
 dyn_coat_4_base [{shapeType=Capsule, length=0, radius=2, shapeName=dyn_coat_4_base_capsule}]
 dyn_coat_5_1 [{
   shapeType=Capsule
   length=11.2721
   radius=5.5003
   shapeName=dyn_coat_5_1_capsule
   center=<9.4339 -0.9315 0.7843>
   rotation=<rot 0.9218 -0.8419 -170.7222>
  }]
 dyn_coat_5_2 [{
   shapeType=Capsule
   length=10.9202
   radius=8.8132
   shapeName=dyn_coat_5_2_capsule
   center=<8.1112 -3.4449 1.4794>
   rotation=<rot -0.8663 -0.1117 -177.575>
  }]
 dyn_coat_5_3 [{
   shapeType=Capsule
   length=19.6771
   radius=11.3901
   shapeName=dyn_coat_5_3_capsule
   center=<10.4655 -4.1598 0.4888>
   rotation=<rot -0.7839 5.3207 -11.8954>
  }]
 dyn_coat_5_4 [{
   shapeType=Capsule
   length=23.8025
   radius=13.102
   shapeName=dyn_coat_5_4_capsule
   center=<12.0225 -1.6967 0.3306>
   rotation=<rot -1.714 -1.6004 141.512>
  }]
 dyn_coat_5_base [{shapeType=Capsule, length=0, radius=2, shapeName=dyn_coat_5_base_capsule}]
 dyn_coat_6_1 [{
   shapeType=Capsule
   length=13.647
   radius=4.6976
   shapeName=dyn_coat_6_1_capsule
   center=<8.3409 -0.146 -0.6544>
   rotation=<rot 0.0001 0.0006 60.0001>
  }]
 dyn_coat_6_2 [{
   shapeType=Capsule
   length=17.3535
   radius=7.4844
   shapeName=dyn_coat_6_2_capsule
   center=<10.4472 -0.3321 -2.0936>
   rotation=<rot 0 0.0003 -100>
  }]
 dyn_coat_6_3 [{shapeType=Capsule, length=26.25, radius=7.75, shapeName=dyn_coat_6_3_capsule, center=<9.759 -0.0817 -3.4919>, rotation=<rot 0 0.0002 70>}]
 dyn_coat_6_4 [{shapeType=Capsule, length=19.9387, radius=8.5608, shapeName=dyn_coat_6_4_capsule, center=<9.9459 0.2326 -3.4008>, rotation=<rot 0 0.0002 70>}]
 dyn_coat_6_base [{shapeType=Capsule, length=0, radius=2, shapeName=dyn_coat_6_base_capsule}]
 dyn_coat_7_1 [{shapeType=Capsule, length=5.5011, radius=5.1534, shapeName=dyn_coat_7_1_capsule, center=<11.3242 0.9507 -2.315>, rotation=<rot 0 0.0003 40>}]
 dyn_coat_7_2 [{shapeType=Capsule, length=4.8365, radius=5.8762, shapeName=dyn_coat_7_2_capsule, center=<8.2309 -0.8881 -3.7451>, rotation=<rot 0 0.0002 50>}]
 dyn_coat_7_3 [{
   shapeType=Capsule
   length=8.9524
   radius=6.5644
   shapeName=dyn_coat_7_3_capsule
   center=<10.0661 -1.8474 -4.9242>
   rotation=<rot 0 0.0002 40>
  }]
 dyn_coat_7_4 [{
   shapeType=Capsule
   length=11.0588
   radius=7.2771
   shapeName=dyn_coat_7_4_capsule
   center=<8.6178 -1.5624 -6.3574>
   rotation=<rot 0 0.0002 40>
  }]
 dyn_coat_7_5 [{
   shapeType=Capsule
   length=11.1339
   radius=7.2453
   shapeName=dyn_coat_7_5_capsule
   center=<8.7509 -2.7477 -7.1239>
   rotation=<rot 0 0.0002 40>
  }]
 dyn_coat_7_base [{shapeType=Capsule, length=0, radius=2, shapeName=dyn_coat_7_base_capsule}]
 foot_l [{
   shapeType=Capsule
   length=21.6706
   radius=11.074
   shapeName=foot_l_capsule
   center=<-2.5655 3.8329 -0.3158>
   rotation=<rot -0.167 3.46 -89.3768>
  }]
 foot_r [{
   shapeType=Capsule
   length=21.6695
   radius=11.1373
   shapeName=foot_r_capsule
   center=<2.4929 -3.8225 0.3154>
   rotation=<rot -0.1677 3.5482 -89.3741>
  }]
 pelvis [{
   shapeType=Capsule
   length=10.2223
   radius=10.473
   shapeName=pelvis_capsule
   center=<-0.8136 -0.325 -0.3993>
   rotation=<rot 0.033 -90.2071 -0.4405>
  }]
 root [{shapeType=Capsule, length=0, radius=214.005, shapeName=root_capsule, center=<0 -1.5038 -205.0303>, rotation=<rot 0 -90 0>}]
 spine_01 [{shapeType=Capsule, length=10, radius=9.5, shapeName=spine_01_capsule, center=<5.0973 0 0.1467>, rotation=<rot -3.6196 -90 -0.0922>}]
 spine_04 [{
   shapeType=Capsule
   length=11.4062
   radius=11.8361
   shapeName=spine_04_capsule
   center=<-4.4366 -0.6386 0.0676>
   rotation=<rot -1.0344 -90 -0.8334>
  }]
 thigh_l [{
   shapeType=Capsule
   length=31.099
   radius=8.8286
   shapeName=thigh_l_capsule
   center=<-19.6949 -0.9531 0.0155>
   rotation=<rot -0.0001 94.3067 -89.452>
  }]
 thigh_r [{
   shapeType=Capsule
   length=32.2967
   radius=8.8033
   shapeName=thigh_r_capsule
   center=<19.6656 1.5792 0.4701>
   rotation=<rot -0.0001 91.0489 -90.7869>
  }]
}
constraints [
 {bone1Name=dyn_coat_2_1, bone2Name=dyn_coat_2_base, swing1LimitDegrees=30, swing2LimitDegrees=30, twistLimitDegrees=2}
 {bone1Name=dyn_coat_2_2, bone2Name=dyn_coat_2_1, swing1LimitDegrees=15, swing2LimitDegrees=30, twistLimitDegrees=2}
 {bone1Name=dyn_coat_2_3, bone2Name=dyn_coat_2_2, swing1LimitDegrees=30, swing2LimitDegrees=45, twistLimitDegrees=2}
 {bone1Name=dyn_coat_2_4, bone2Name=dyn_coat_2_3, swing1LimitDegrees=15, swing2LimitDegrees=45, twistLimitDegrees=2}
 {bone1Name=dyn_coat_3_1, bone2Name=dyn_coat_3_base, swing1LimitDegrees=25, swing2LimitDegrees=45, twistLimitDegrees=2}
 {bone1Name=dyn_coat_3_2, bone2Name=dyn_coat_3_1, swing1LimitDegrees=15, swing2LimitDegrees=45, twistLimitDegrees=2}
 {bone1Name=dyn_coat_3_3, bone2Name=dyn_coat_3_2, swing1LimitDegrees=25, swing2LimitDegrees=35, twistLimitDegrees=2}
 {bone1Name=dyn_coat_3_4, bone2Name=dyn_coat_3_3, swing1LimitDegrees=30, swing2LimitDegrees=45, twistLimitDegrees=2}
 {bone1Name=dyn_coat_4_1, bone2Name=dyn_coat_4_base, swing1LimitDegrees=30, swing2LimitDegrees=45, twistLimitDegrees=2}
 {bone1Name=dyn_coat_4_2, bone2Name=dyn_coat_4_1, swing1LimitDegrees=10, swing2LimitDegrees=45, twistLimitDegrees=2}
 {bone1Name=dyn_coat_4_3, bone2Name=dyn_coat_4_2, swing1LimitDegrees=15, swing2LimitDegrees=35, twistLimitDegrees=2}
 {bone1Name=dyn_coat_4_4, bone2Name=dyn_coat_4_3, swing1LimitDegrees=30, swing2LimitDegrees=45, twistLimitDegrees=2}
 {bone1Name=dyn_coat_5_2, bone2Name=dyn_coat_5_1, swing1LimitDegrees=25, swing2LimitDegrees=45, twistLimitDegrees=2}
 {bone1Name=dyn_coat_5_3, bone2Name=dyn_coat_5_2, swing1LimitDegrees=25, swing2LimitDegrees=45, twistLimitDegrees=2}
 {bone1Name=dyn_coat_5_4, bone2Name=dyn_coat_5_3, swing1LimitDegrees=30, swing2LimitDegrees=45, twistLimitDegrees=2}
 {bone1Name=dyn_coat_6_1, bone2Name=dyn_coat_6_base, swing1LimitDegrees=45, swing2LimitDegrees=30, twistLimitDegrees=2}
 {bone1Name=dyn_coat_6_2, bone2Name=dyn_coat_6_1, swing1LimitDegrees=30, swing2LimitDegrees=30, twistLimitDegrees=2}
 {bone1Name=dyn_coat_6_3, bone2Name=dyn_coat_6_2, swing1LimitDegrees=30, swing2LimitDegrees=45, twistLimitDegrees=2}
 {bone1Name=dyn_coat_6_4, bone2Name=dyn_coat_6_3, swing1LimitDegrees=30, swing2LimitDegrees=45, twistLimitDegrees=2}
 {bone1Name=dyn_coat_1_1, bone2Name=dyn_coat_1_base, swing1LimitDegrees=30, swing2LimitDegrees=30, twistLimitDegrees=2}
 {bone1Name=dyn_coat_1_2, bone2Name=dyn_coat_1_1, swing1LimitDegrees=30, swing2LimitDegrees=30, twistLimitDegrees=2}
 {bone1Name=dyn_coat_1_3, bone2Name=dyn_coat_1_2, swing1LimitDegrees=15, swing2LimitDegrees=45, twistLimitDegrees=2}
 {bone1Name=dyn_coat_1_4, bone2Name=dyn_coat_1_3, swing1LimitDegrees=15, swing2LimitDegrees=45, twistLimitDegrees=2}
 {bone1Name=dyn_coat_1_5, bone2Name=dyn_coat_1_4, swing1LimitDegrees=15, swing2LimitDegrees=45, twistLimitDegrees=2}
 {bone1Name=dyn_coat_7_1, bone2Name=dyn_coat_7_base, swing1LimitDegrees=30, swing2LimitDegrees=30, twistLimitDegrees=2}
 {bone1Name=dyn_coat_7_2, bone2Name=dyn_coat_7_1, swing1LimitDegrees=30, swing2LimitDegrees=30, twistLimitDegrees=2}
 {bone1Name=dyn_coat_7_3, bone2Name=dyn_coat_7_2, swing1LimitDegrees=20, swing2LimitDegrees=60, twistLimitDegrees=2}
 {bone1Name=dyn_coat_7_4, bone2Name=dyn_coat_7_3, swing1LimitDegrees=30, swing2LimitDegrees=45, twistLimitDegrees=2}
 {bone1Name=dyn_coat_7_5, bone2Name=dyn_coat_7_4, swing1LimitDegrees=30, swing2LimitDegrees=45, twistLimitDegrees=2}
 {
  bone1Name=dyn_coat_2_4
  bone2Name=dyn_coat_1_5
  swing1LimitDegrees=45
  swing1Motion=Free
  swing2LimitDegrees=45
  swing2Motion=Free
  twistLimitDegrees=45
  twistMotion=Free
 }
 {
  bone1Name=dyn_coat_2_4
  bone2Name=dyn_coat_3_4
  swing1LimitDegrees=45
  swing1Motion=Free
  swing2LimitDegrees=45
  swing2Motion=Free
  twistLimitDegrees=45
  twistMotion=Free
 }
 {
  bone1Name=dyn_coat_3_4
  bone2Name=dyn_coat_4_4
  swing1LimitDegrees=45
  swing1Motion=Free
  swing2LimitDegrees=45
  swing2Motion=Free
  twistLimitDegrees=45
  twistMotion=Free
 }
 {
  bone1Name=dyn_coat_5_4
  bone2Name=dyn_coat_4_4
  swing1LimitDegrees=45
  swing1Motion=Free
  swing2LimitDegrees=45
  swing2Motion=Free
  twistLimitDegrees=45
  twistMotion=Free
 }
 {
  bone1Name=dyn_coat_6_4
  bone2Name=dyn_coat_5_4
  swing1LimitDegrees=45
  swing1Motion=Free
  swing2LimitDegrees=45
  swing2Motion=Free
  twistLimitDegrees=45
  twistMotion=Free
 }
 {
  bone1Name=dyn_coat_7_5
  bone2Name=dyn_coat_6_4
  swing1LimitDegrees=45
  swing1Motion=Free
  swing2LimitDegrees=45
  swing2Motion=Free
  twistLimitDegrees=45
  twistMotion=Free
 }
 {
  bone1Name=dyn_coat_2_2
  bone2Name=dyn_coat_1_3
  swing1LimitDegrees=45
  swing1Motion=Free
  swing2LimitDegrees=45
  swing2Motion=Free
  twistLimitDegrees=45
  twistMotion=Free
 }
 {
  bone1Name=dyn_coat_6_2
  bone2Name=dyn_coat_7_3
  swing1LimitDegrees=45
  swing1Motion=Free
  swing2LimitDegrees=45
  swing2Motion=Free
  twistLimitDegrees=45
  twistMotion=Free
 }
 {bone1Name=dyn_coat_5_1, bone2Name=dyn_coat_5_base, swing1LimitDegrees=30, swing2LimitDegrees=30, twistLimitDegrees=2}
 {
  bone1Name=dyn_coat_5_2
  bone2Name=dyn_coat_6_2
  swing1LimitDegrees=45
  swing1Motion=Free
  swing2LimitDegrees=45
  swing2Motion=Free
  twistLimitDegrees=45
  twistMotion=Free
 }
 {
  bone1Name=dyn_coat_6_base
  bone2Name=pelvis
  swing1LimitDegrees=45
  swing1Motion=Free
  swing2LimitDegrees=45
  swing2Motion=Free
  twistLimitDegrees=45
  twistMotion=Free
 }
 {
  bone1Name=dyn_coat_5_base
  bone2Name=pelvis
  swing1LimitDegrees=45
  swing1Motion=Free
  swing2LimitDegrees=45
  swing2Motion=Free
  twistLimitDegrees=45
  twistMotion=Free
 }
 {
  bone1Name=dyn_coat_4_base
  bone2Name=pelvis
  swing1LimitDegrees=45
  swing1Motion=Free
  swing2LimitDegrees=45
  swing2Motion=Free
  twistLimitDegrees=45
  twistMotion=Free
 }
 {
  bone1Name=dyn_coat_3_base
  bone2Name=pelvis
  swing1LimitDegrees=45
  swing1Motion=Free
  swing2LimitDegrees=45
  swing2Motion=Free
  twistLimitDegrees=45
  twistMotion=Free
 }
 {
  bone1Name=dyn_coat_2_base
  bone2Name=pelvis
  swing1LimitDegrees=45
  swing1Motion=Free
  swing2LimitDegrees=45
  swing2Motion=Free
  twistLimitDegrees=45
  twistMotion=Free
 }
 {
  bone1Name=dyn_coat_1_4
  bone2Name=dyn_coat_2_3
  swing1LimitDegrees=45
  swing1Motion=Free
  swing2LimitDegrees=45
  swing2Motion=Free
  twistLimitDegrees=45
  twistMotion=Free
 }
 {
  bone1Name=dyn_coat_6_3
  bone2Name=dyn_coat_7_4
  swing1LimitDegrees=45
  swing1Motion=Free
  swing2LimitDegrees=45
  swing2Motion=Free
  twistLimitDegrees=45
  twistMotion=Free
 }
 {
  bone1Name=dyn_coat_6_3
  bone2Name=dyn_coat_5_3
  swing1LimitDegrees=45
  swing1Motion=Free
  swing2LimitDegrees=45
  swing2Motion=Free
  twistLimitDegrees=45
  twistMotion=Free
 }
 {
  bone1Name=dyn_coat_3_3
  bone2Name=dyn_coat_2_3
  swing1LimitDegrees=45
  swing1Motion=Free
  swing2LimitDegrees=45
  swing2Motion=Free
  twistLimitDegrees=45
  twistMotion=Free
 }
 {
  bone1Name=dyn_coat_2_2
  bone2Name=dyn_coat_3_2
  swing1LimitDegrees=45
  swing1Motion=Free
  swing2LimitDegrees=45
  swing2Motion=Free
  twistLimitDegrees=45
  twistMotion=Free
 }
 {
  bone1Name=dyn_coat_5_3
  bone2Name=dyn_coat_4_3
  swing1LimitDegrees=45
  swing1Motion=Free
  swing2LimitDegrees=45
  swing2Motion=Free
  twistLimitDegrees=45
  twistMotion=Free
 }
 {
  bone1Name=dyn_coat_3_3
  bone2Name=dyn_coat_4_3
  swing1LimitDegrees=45
  swing1Motion=Free
  swing2LimitDegrees=45
  swing2Motion=Free
  twistLimitDegrees=45
  twistMotion=Free
 }
 {
  bone1Name=dyn_coat_6_1
  bone2Name=dyn_coat_7_2
  swing1LimitDegrees=45
  swing1Motion=Free
  swing2LimitDegrees=45
  swing2Motion=Free
  twistLimitDegrees=45
  twistMotion=Free
 }
 {
  bone1Name=dyn_coat_1_2
  bone2Name=dyn_coat_2_1
  swing1LimitDegrees=45
  swing1Motion=Free
  swing2LimitDegrees=45
  swing2Motion=Free
  twistLimitDegrees=45
  twistMotion=Free
 }
]
