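# Post-process-style cosmetic AnimBP for the M_MED_CombCrater player body, adding secondary motion to the pose copied from the attached parent mesh.
# A Control Rig takes movement, crouch, slide, skydive, carried and physics-disabled inputs, then feeds three chained spine_01-space RigidBody nodes with separate physics assets, including skydiving and front-end.
# The event graph derives those states, plus alphas and gravity override, using deprecated helpers.
# A fourth RigidBody node is unconnected.
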AnimBlueprint /BRCosmetics/Characters/Player/Male/Medium/Bodies/M_MED_CombCrater/Meshes/M_MED_CombCrater_AnimBP
target_skeleton=&/M_MED_CombCrater_Skeleton
graphs [
 AnimGraph {nodes [
   RigidBody!(ComponentPose, Alpha, OverrideWorldGravity:Vector="0,0,0", ExternalForce) -> (Pose) {
    bOverrideWorldGravity=true
    overridePhysicsAsset=&/M_MED_CombCrater_Physics
    simulationSpace=BaseBoneSpace
    alphaBoolBlend {blendOption=Linear}
    baseBoneRef {boneName=spine_01}
    componentAppliedLinearAccClamp=<0 0 0>
    simSpaceSettings {
     maxAngularAcceleration=30
     maxAngularVelocity=4
     maxLinearAcceleration=1000
     maxLinearVelocity=300
     velocityScaleZ=0.25
     worldAlpha=0.75
     externalLinearDragV=<2 6 2>
    }
   }
   n1 = Control Rig!(
    Source=Copy Pose From Mesh() {bUseAttachedParent=true}
    Alpha
    Dyn_baseball_sliding:double=Get Dyn Baseball Sliding() -> (Dyn Baseball Sliding:double)
    Dyn_knee_sliding:double=Get Dyn Knee Sliding() -> (Dyn Knee Sliding:double)
    Dyn_Skydive:double=Get Dyn Skydive() -> (Dyn Skydive:double)
    Movement Alpha:double=Get MovementAlpha() -> (MovementAlpha:double)
    is_DBNO_Carried:bool=Get bIsDBNOCarried()
    RBAN:double=Get RBAN_float() -> (RBAN_float:double)
    Dyn_Crouch:double=Get Dyn Crouch() -> (Dyn Crouch:double)
   ) -> (Pose) {
    alphaBoolBlend {blendOption=Linear}
    controlRigAssetReference {blueprintRigClass=&/M_MED_CombCrater_CtrlRig.M_MED_CombCrater_CtrlRig_C}
   }
   n2 = RigidBody!(ComponentPose=Local To Component(LocalPose=@n1), bAlphaBoolEnabled:bool, OverrideWorldGravity="0,0,0", ExternalForce) -> (Pose) {
    alphaInputType=Bool
    bOverrideWorldGravity=true
    overridePhysicsAsset=&/M_MED_CombCrater_Physics_01
    simulationSpace=BaseBoneSpace
    alphaBoolBlend {blendInTime=0.5, blendOption=Linear, blendOutTime=0.2}
    baseBoneRef {boneName=spine_01}
    componentAppliedLinearAccClamp=<0 0 0>
    simSpaceSettings {
     maxAngularAcceleration=30
     maxAngularVelocity=4
     maxLinearAcceleration=1250
     maxLinearVelocity=400
     velocityScaleZ=0.25
     worldAlpha=0.75
     externalLinearDragV=<2 5 2>
    }
   }
   n3 = RigidBody(ComponentPose=@n2, OverrideWorldGravity="0,0,0") {
    alphaInputType=Bool
    bOverrideWorldGravity=true
    overridePhysicsAsset=&/M_MED_CombCrater_Physics_skydiving
    simulationSpace=BaseBoneSpace
    alphaBoolBlend {blendInTime=0.5, blendOption=Linear, blendOutTime=0.2}
    baseBoneRef {boneName=spine_01}
    componentAppliedLinearAccClamp=<0 0 0>
    simSpaceSettings {
     maxAngularAcceleration=30
     maxAngularVelocity=4
     maxLinearAcceleration=1250
     maxLinearVelocity=400
     velocityScaleZ=0.25
     worldAlpha=0.75
     externalLinearDragV=<2 6.5 2>
    }
   }
   n4 = RigidBody(ComponentPose=@n3, OverrideWorldGravity="0,0,0") {
    alphaInputType=Bool
    bOverrideWorldGravity=true
    overridePhysicsAsset=&/M_MED_CombCrater_FrontEnd_Physics
    simulationSpace=BaseBoneSpace
    alphaBoolBlend {blendOption=Linear}
    baseBoneRef {boneName=spine_01}
    componentAppliedLinearAccClamp=<0 0 0>
    simSpaceSettings {
     maxAngularAcceleration=60
     maxAngularVelocity=3
     maxLinearAcceleration=450
     maxLinearVelocity=100
     velocityScaleZ=0.25
     worldAlpha=0.75
     externalLinearDragV=<2 6 2>
    }
   }
   Output Pose(Result=Component To Local(ComponentPose=@n4)) {layerGroup=None}
  ]}
 EventGraph {nodes [
   Event BlueprintUpdateAnimation()
   > n1 = Sequence!(execute) -> (then_0, then_1, then_2:Exec)
   > n2 = Gravity Override State Function DEPRECATED(
    CharacterPartAnimInstance=Self-Reference()
    GameTimeInSeconds=GetGameTimeInSeconds()
    GravityOverrideParams=Get Gravity_Override_Params() -> (Gravity_Override_Params:Gravity Override Params Struct Structure)
   )
   > Set Dyn_Override_World_Gravity(Dyn_Override_World_Gravity:Vector=@n2) -> (Output_Get:Vector)
   n3 = Get bRigidBodyEnabled()
   > Set RBAN_float(RBAN_float:double="To Float (Boolean)"(InBool=NOT Boolean(A=@n3))) -> (Output_Get:double)
   n4 = Get bIsSkyDiving()
   n5 = Get bIsInFrontEnd()
   n6 = Get bIsBaseballSliding()
   > Set level_rigidBody(level_rigidBody:bool=AND Boolean!(A=NOT Boolean(A=@n5), B=NOT Boolean(A=@n6), C:bool=NOT Boolean(A=@n4)) -> (ReturnValue)) -> (Output_Get:bool)
   > n7 = Branch(Condition=Get bIsInFrontEnd())
   > Set Alpha_RB(Alpha_RB:float=0) -> (Output_Get:float)
   > Set Alpha_FE(Alpha_FE:float=1) -> (Output_Get:float)
   Set Alpha_RB(execute=@n7.else, Alpha_RB=1)
   > Set Alpha_FE(Alpha_FE=0)
   n8 = Branch(execute=@n1.then_1, Condition=Get bIsCrouching())
   > Set Dyn Crouch(Dyn Crouch:double=1) -> (Output_Get:double)
   Set Dyn Crouch(execute=@n8.else, Dyn Crouch=0)
   n9 = Player Movement Function DEPRECATED!(
    execute=@n1.then_2
    CharacterPartAnimInstance:Custom Character Part Anim Instance Object Reference=Self-Reference()
    X Dot Product:double=25
    Y Dot Product:double=25
    Z Dot Product:double=25
    DrawDebug:bool=false
    Debug Scalar:double=50
   ) -> (then, X Movement:double, Y Movement:double, Z Movement:double)
   > Set MovementAlpha(MovementAlpha:double="Absolute (Float)"(A=float * float(A=@n9.Y Movement, B=0.045) {is=title K2Node_PromotableOperator})) -> (Output_Get:double)
   Event OnBeginKneeSliding() {is=title K2Node_Event}
   > Set Dyn Knee Sliding(Dyn Knee Sliding:double=1) -> (Output_Get:double)
   Event OnEndKneeSliding() {is=title K2Node_Event}
   > Set Dyn Knee Sliding(Dyn Knee Sliding=0)
   Event OnBeginBaseballSliding() {is=title K2Node_Event}
   > Set Dyn Baseball Sliding(Dyn Baseball Sliding:double=1) -> (Output_Get:double)
   Event OnEndBaseballSliding() {is=title K2Node_Event}
   > Set Dyn Baseball Sliding(Dyn Baseball Sliding=0)
   Event OnBeginSkydiving()
   > Set Dyn Skydive(Dyn Skydive:double="To Float (Boolean)"(InBool=Get bIsSkydiveFloating())) -> (Output_Get:double)
   Event OnEndSkydiving()
   > Set Dyn Skydive(Dyn Skydive=0)
  ]}
]
variables [
 Alpha_FE {type=number, default=1}
 Alpha_RB {type=number, default=1}
 Dyn Baseball Sliding {type=number}
 Dyn Crouch {type=number}
 Dyn Knee Sliding {type=number}
 Dyn Skydive {type=number}
 Dyn_Override_World_Gravity {type=Vector}
 Gravity_Override_Params {
  type=GravityOverrideParamsStruct
  default {
   emoteOrMelee {gravityOverride=<0 0 -1500>}
   onGround_Moving {gravityOverride=<0 0 -1250>, windAmplitude=<150 -500 300>, windFrequency=<3 5 4>}
   skydive_FloatingDown {gravityOverride=<-75 -750 600>, windFrequency=<1.5 5 5.5>}
   skydive_GoingUp {gravityOverride=<-150 -150 600>, windAmplitude=<75 -125 125>, windFrequency=<1 3 4>}
   tactical_Sprint {gravityOverride=<-500 -600 -980>}
  }
 }
 level_rigidBody {type=boolean}
 MovementAlpha {type=number}
 RBAN_float {type=number}
]
